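; NPC charge attack ability, activatable when the blackboard target is within range and no grenade is equipped.
; After a delay, an intro montage notify starts the dash, with continuous target selection and a run timer.
; Blocking hits, optionally ignoring walkable landscape, trigger pawn or world finishers applying effect containers; misses play a montage with optional follow-through damage.
; Impulse, falling, shock knockback, swimming and a fail-safe timer can end the charge.
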(event Ability|EventActivateAbility
  (bind _returnvalue (CallFunction|GetAvatarasFortPlayerPawn))
  ;[1 1 1] Activate Ability
  (Utilities|FlowControl|Delay 0.7) ; To make sure AI has enough time to stop BT
  (Variables|Default|SetHitSomething false)
  (bind _returnvalue_2 (Ability|CommitAbility))
  (if _returnvalue_2
    (CallFunction|InitAbility)
    (Camera|PopCameraMode)
    (Fortnite|KeepCGPathOptimizations (Variables|Default|GetFortPlayerController))
    (CallFunction|DashIntro)
    ;[1 1 1] Cancel Charge if these events / tags are active on AI
    (Ability|Tasks|WaitGameplayEvent "(TagName=\"Ability.CancelOnImpulse\")"
      (:then
        (Ability|Tasks|WaitGameplayTagAdd "(TagName=\"MovementMode.Falling\")" _returnvalue
          (:then
            (Ability|Tasks|WaitGameplayTagAdd "(TagName=\"Granted.Athena.Item.ShockGrenade.Knocked\")" _returnvalue
              (:then
                ;[1 1 1] Fail Safe to make sure Ability Doesn't Get Jammed and AI keeps playing nice
                (Utilities|FlowControl|Delay (ScalableFloat|GetValueatLevel (Variables|Default|GetSF_FailSafeTimer))) ; If After [2]s the Dash didn't start - we are probably stuck here, so stop ability.
                (Development|PrintString "Change Ability FailSafe Was Called" true true "(R=1.000000,G=0.101204,B=0.170911,A=1.000000)" 2e+01)
                (CallFunction|Cleanup))
              (:Added
                _)))
          (:Added
            _))
        (CallFunction|Cleanup))
      (:EventReceived
        (CallFunction|Cleanup)))
    (else
      (Ability|EndAbility))))

(event Custom|DashIntro
  ;[1 1 1] Dash Intro
  (Ability|Tasks|PlayMontageAndWaitNotify "None" (Variables|Montages|GetM_DashIntro) (Variables|Default|GetDashPlayrate) "Intro"
    (:OnNotifyBegin
      (CallFunction|Dash))))

(event Custom|Dash
  ;[1 1 1] Execute Dash
  (Development|PrintString "Dash")
  (Ability|Tasks|WaitGameplayTagAdd "(TagName=\"MovementMode.Falling\")"
    (:then
      (Ability|Tasks|WaitGameplayTagAdd "(TagName=\"Gameplay.Action.Player.SurfaceSwimming\")"
        (:then
          (Networking|ForceNetUpdate (Variables|Default|GetPlayerPawn))
          (bind (_asynctaskproxy _targetdata) (Ability|Tasks|DoTargetSelectionContinuous (Variables|Default|GetDashTargetSelectionTaskName) "(TagName=\"Gameplay.Melee.Hit1\")" false)
            (:then
              (bind _returnvalue (Utilities|Time|SetTimerbyEvent (EventDispatchers|CreateEvent :!event Custom|HitNothing) (ScalableFloat|GetValueatLevel (Variables|Default|GetRunTimer))))
              (Variables|Default|SetRunTimerHandle _returnvalue))
            (:Targeted
              (if (CallFunction|IsBlockingHit _targetdata)
                (Variables|Default|SetHitSomething true)
                (CallFunction|PlayImpactGC _targetdata)
                (Ability|RemoveGameplayCueFromOwner (Variables|Default|GetGC_DashLoop))
                (CallFunction|FinisherAttack _targetdata)))))
        (:Added
          (CallFunction|HitNothing))))
    (:Added
      (CallFunction|HitNothing))))

(event Custom|HitNothing
  ;[1 1 1] Hit Nothing
  (if (not (Variables|Default|GetHitSomething))
    (Ability|Tasks|PlayMontageAndWaitNotify "None" (Variables|Default|GetMontageHitNothing) 1.0 "None" false
      (:then
        (Ability|EndTaskbyInstanceName (Variables|Default|GetDashTargetSelectionTaskName))
        (if (Variables|Default|GetUseFollowThroughDamage)
          (bind (_asynctaskproxy _targetdata) (Ability|Tasks|DoTargetSelectionContinuous (Variables|Default|GetMissTargetSelectionTaskName) "(TagName=\"Gameplay.Melee.Hit1\")")
            (:Targeted
              (CallFunction|PlayImpactGC _targetdata)
              (CallFunction|HandleFinisherNotify "(TagName=\"Gameplay.Melee.Hit1\")")
              (Ability|EndTaskbyInstanceName (Variables|Default|GetMissTargetSelectionTaskName))))))
      (:OnCompleted
        (CallFunction|Cleanup))
      (:OnNotifyBegin
        (if (Variables|Default|GetUseFollowThroughDamage)
          (Ability|EndTaskbyInstanceName (Variables|Default|GetMissTargetSelectionTaskName)))))))

(event Custom|Cleanup
  ;[1 1 1] Execute Dash
  (CallFunction|CleanUpDash) ; tags to drive umbrella animation
  (Ability|EndAbility))

(event Custom|FinisherAttack
  (params (InputPin :type Gameplay%20Ability%20Target%20Data%20Handle%20Structure%20%28by%20ref%29) (ForceHitPlayerAnim :type Boolean))
  ;[1 1 1] Finisher
  (Utilities|Time|ClearandInvalidateTimerbyHandle (Variables|Default|GetRunTimerHandle))
  (Networking|ForceNetUpdate (Variables|Default|GetPlayerPawn))
  (Ability|EndTaskbyInstanceName (Variables|Default|GetDashTargetSelectionTaskName))
  (Camera|PopCameraMode)
  (bind _montage (CallFunction|SelectHitMontage InputPin ForceHitPlayerAnim))
  (Ability|Tasks|PlayMontageandWaitforTrigger "None" _montage)
  (CallFunction|HandleFinisherNotify (Variables|Default|GetTag_FinisherHitNotify))
  (Utilities|Time|SetTimerbyEvent (delegate Custom|CleanUpAfterHit) (Animation|GetPlayLength _montage)))

(event Custom|CleanUpAfterHit
  ;[1 1 1] Finisher
  (CallFunction|Cleanup))

(event Custom|HandleFinisherNotify
  (params (Tag :type Gameplay%20Tag%20Structure))
  ;[1 1 1] Technically could handle any passed-in effect container tag, but specifically meant to handle the two hits of the finisher move
  (bind (_asynctaskproxy _targetdata _applicationtag) (Ability|Tasks|DoTargetSelection "None" Tag)
    (:Targeted
      (Ability|ApplyGameplayEffectContainer _targetdata _applicationtag (Math|Float|Truncate (Variables|Default|GetGELevel))))))

(event Ability|EventOnEndAbility
  (params (bWasCancelled :type Boolean))
  ;[1 1 1] Reset hit bool.  Needs to be reset despite being instance on execution for some reason
  (|CallParentFunction bWasCancelled)
  (Variables|Default|SetHitSomething false)
  (Camera|PopCameraMode))

(fn CleanUpDash
  (Ability|RemoveGameplayCueFromOwner (Variables|Default|GetGC_DashLoop))
  (Networking|ForceNetUpdate (Variables|Default|GetPlayerPawn))
  (Ability|EndTaskbyInstanceName (Variables|Default|GetDashTargetSelectionTaskName))
  (Ability|EndTaskbyInstanceName (Variables|Default|GetMissTargetSelectionTaskName)))

(fn ImpulseHitDashTarget
  (params (TargetData :type Gameplay%20Ability%20Target%20Data%20Handle%20Structure)))

(fn InitAbility
  (bind _asfort_player_pawn (Utilities|Casting|CastToFortPlayerPawn (Ability|GetAvatarActorfromActorInfo))
    (:then
      (Variables|Default|SetPlayerPawn _asfort_player_pawn)
      (Variables|Default|SetFortPlayerController (Fortnite|GetFortPlayerControllerfromActor (Ability|GetAvatarActorfromActorInfo)))
      (return))
    (:CastFailed
      (CallFunction|Cleanup))))

(fn IsBlockingHit
  (params (TargetData :type Gameplay%20Ability%20Target%20Data%20Handle%20Structure))
  (bind (_returnvalue_bblockinghit _returnvalue_binitialoverlap _returnvalue_time _returnvalue_distance _returnvalue_location _returnvalue_impactpoint _returnvalue_normal _returnvalue_impactnormal _returnvalue_physmat _returnvalue_hitactor) (Collision|BreakHitResult (Ability|TargetData|GetHitResultfromTargetData TargetData)))
  (bind (_returnvalue_impactnormal_x _returnvalue_impactnormal_y _returnvalue_impactnormal_z) (Math|Vector|BreakVector _returnvalue_impactnormal))
  (if (ScalableFloat|GetValueasBool (Variables|Default|GetContinueChargeOnLandscapeSlopes))
    (Utilities|Casting|CastToLandscape _returnvalue_hitactor
      (:then
        (if (>= _returnvalue_impactnormal_z (ScalableFloat|GetValueatLevel (Variables|Default|GetLandscapeSlopeMinWalkableZ)))
          (return false)
          (else
            _)))
      (:CastFailed
        _))
    (return true)
    (else
      _)))

(fn CanActivateAbility
  (params (ActorInfo :type Gameplay%20Ability%20Actor%20Info%20Structure) (Handle :type Gameplay%20Ability%20Spec%20Handle%20Structure))
  (bind _returnvalue (CallFunction|GetAvatarasFortPlayerPawn))
  (bind _relevanttags (|CallParentFunction ActorInfo Handle))
  (bind _targetactor (Class|BPFLAICommonFunctions|GetTargetActorfromBlackboard _returnvalue))
  (return _relevanttags (and (<= (Math|Float|Absolute(Float) (Math|Vector|Distance(Vector) (Transformation|GetActorLocation _targetactor) (Transformation|GetActorLocation _returnvalue))) (ScalableFloat|GetValueatLevel (Variables|Default|GetMaximumDistance))) (not (GameplayTags|HasTag (Tags|GetWeaponDataGameplayTags (Variables|/Script/FortniteGame.FortPawn|GetCurrentWeapon _returnvalue)) "(TagName=\"Item.Consumable.Grenade\")" true)))))

(fn OnRep_ChargeInterval)

(fn OnRep_ChargeTime)

(fn PlayImpactGC
  (params (TargetData :type Gameplay%20Ability%20Target%20Data%20Handle%20Structure))
  (bind (_bblockinghit _binitialoverlap _time _distance _location _impactpoint _normal _impactnormal _physmat) (Collision|BreakHitResult (Ability|TargetData|GetHitResultfromTargetData TargetData)))
  (Ability|ExecuteGameplayCueWithParamsOnOwner (Variables|Default|GetGC_DashImpact) (Ability|GameplayCue|MakeGameplayCueParameters :NormalizedMagnitude 0.0 :RawMagnitude 0.0 :EffectContext (Ability|GetContextfromOwner TargetData) :Location _location :Normal _normal :PhysicalMaterial _physmat :GameplayEffectLevel 1 :AbilityLevel 1 :bReplicateLocationWhenUsingMinimalRepProxy false)))

(fn SelectHitMontage
  (params (TargetData :type Gameplay%20Ability%20Target%20Data%20Handle%20Structure) (ForceHitPlayer :type Boolean))
  (bind _returnvalue (Ability|TargetData|GetHitResultfromTargetData TargetData))
  (bind (_bblockinghit _binitialoverlap _time _distance _location _impactpoint _normal _impactnormal _physmat _hitactor) (Collision|BreakHitResult _returnvalue))
  (Development|PrintString (Utilities|String|Append "Force hit player anim: " (Utilities|String|ToString(Boolean) ForceHitPlayer)))
  (if ForceHitPlayer
    (return (Variables|Montages|GetM_HitPlayer))
    (else
      (Utilities|IsValid _hitactor
        (:Is%20Valid
          (bind _asfort_pawn (Utilities|Casting|CastToFortPawn _hitactor)
            (:then
              (if (not (Health|IsDBNO _asfort_pawn))
                (return (Variables|Montages|GetM_HitPlayer))))
            (:CastFailed
              _))
          (return (Variables|Montages|GetM_HitWorld)))
        (:Is%20Not%20Valid
          (return (Variables|Montages|GetM_HitWorld)))))))
